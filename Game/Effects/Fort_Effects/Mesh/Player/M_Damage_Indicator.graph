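# Unlit, two-sided translucent material for mesh particles, which by name and path draws a player damage indicator.
# Particle color RGB runs through the shared FX emissive adjustment function.
# Particle alpha is multiplied by the biased green channel of a smooth reversed ramp texture, with dynamic parameter Param1 sliding the ramp along V.
# A hit-marker ramp sample and a spare constant sit unconnected in the graph.
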Material /Game/Effects/Fort_Effects/Mesh/Player/M_Damage_Indicator
graph {
 properties {
  bUsedWithMeshParticles=true
  blendMode=BLEND_Translucent
  pixelDepthOffsetMode=PDOM_Legacy
  shadingModel=MSM_Unlit
  twoSided=true
  usedShadingModels=Unlit
 }
 nodes [
  Constant() {r=0.4}
  n1 = ParticleColor()
  MaterialFunctionCall(Emissive=@n1.RGB) {materialFunction=&MF_FX_EmissiveAdjustments}
  n2 = Add(A=TextureCoordinate(), B=AppendVector(A=Constant(), B=DynamicParameter() {defaultValue=<rgba 0 0 0 1>}.Param1))
  Multiply(A=@n1.A, B=Add(A=TextureSample(UVs=@n2) {texture=/Game/Effects/Fort_Effects/Textures/Ramps/T_Ramp_Smooth_Reverse}.G, B=Constant() {r=0.5}))
  TextureSample() {texture=/Game/Effects/Fort_Effects/Textures/T_Active_Reload_Hit_Marker_Ramp}
 ]
}
